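# Material function for the Big01 cosmetic wheel that outputs animated, noise-distorted UVs.
# Two samples of the shared noise heightmap pan over tiled coordinates in opposing diagonal directions at different speeds and are summed.
# The red and green channels of that sum are scaled down and added to U-tiled base coordinates to form the UV output.
# Both samples use shared wrap samplers.
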MaterialFunction /VehicleCosmetics/Wheels/Big01/Materials/MF_Wheel_Big01_UVPanDist
graph {nodes [
  n1 = TextureCoordinate() {uTiling=2, vTiling=2}
  n2 = Panner(Coordinate=@n1) {speedX=-0.5, speedY=0.5}
  n3 = Panner(Coordinate=@n1) {speedX=0.5, speedY=-1}
  n4 = ComponentMask(None=Add(
    A=TextureSample(UVs=@n2) {samplerSource=SSM_Wrap_WorldGroupSettings, texture=/VehicleCosmetics/Wheels/_SharedTextures/FX/T_Noise_Heightmap}.RGB
    B=TextureSample(UVs=@n3) {samplerSource=SSM_Wrap_WorldGroupSettings, texture=&T_Noise_Heightmap}.RGB
   )) {g=true, r=true}
  FunctionOutput(None=Add(A=Multiply(A=@n4) {constB=0.1}, B=TextureCoordinate() {uTiling=2})) {outputName=UV}
 ]}
