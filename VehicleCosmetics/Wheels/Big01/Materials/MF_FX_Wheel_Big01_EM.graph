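# Material function for the Big01 vehicle wheel cosmetic that outputs animated FX emission, base color, normal and base spec mask.
# Time-driven panners scroll noise and energy textures over UVs from MF_Wheel_Big01_UVPanDist, forming the glow that rerouted inputs shape and tint.
# A tiled zebra normal is angle-corrected blended into the normal, and a faint Fresnel lifts base color.
# Named reroutes carry the packed mask channels and tuning inputs throughout.
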MaterialFunction /VehicleCosmetics/Wheels/Big01/Materials/MF_FX_Wheel_Big01_EM
graph {nodes [
  n1 = NamedRerouteUsage()
  n2 = MaterialFunctionCall() {materialFunction=&/MF_Wheel_Big01_UVPanDist}
  n3 = Multiply(A=Time(), B=NamedRerouteUsage())
  n4 = TextureSample(UVs=Panner(Coordinate=@n2.UV, Time=@n3) {speedX=0.25, speedY=0.6}) {samplerSource=SSM_Wrap_WorldGroupSettings, texture=/VehicleCosmetics/Wheels/_SharedTextures/FX/T_NoiseText}
  n5 = Power(Base=Multiply(A=Multiply(A=@n1) {constB=2}, B=Clamp(None=@n4.G) {minDefault=0.075}) {constB=2}) {constExponent=2.75}
  n6 = Panner(Coordinate=@n2.UV, Time=@n3) {speedX=0.25, speedY=0.8}
  n7 = Multiply(A=Multiply(
    A=Power(Base=@n4.G) {constExponent=1.5}
    B=TextureSample(UVs=@n6) {samplerSource=SSM_Wrap_WorldGroupSettings, texture=/VehicleCosmetics/Wheels/_SharedTextures/FX/T_Energy}.G
   )) {constB=4}
  n8 = NamedRerouteUsage()
  n9 = NamedRerouteUsage()
  n10 = Add(A=NamedRerouteUsage(), B=Multiply(A=Multiply(A=@n8, B=@n9)) {constB=0.5})
  n11 = NamedRerouteUsage()
  FunctionOutput(None=Multiply(A=Add(A=@n5, B=@n7), B=Multiply(A=@n10, B=@n11))) {outputName=Emission}
  n12 = NamedRerouteUsage()
  n13 = Multiply(A=Fresnel(Normal=NamedRerouteUsage()) {exponent=3}) {constB=0.05}
  FunctionOutput(None=LinearInterpolate(A=@n12, B=Add(A=@n12, B=@n13), Alpha=NamedRerouteUsage())) {outputName=Base Color, sortPriority=-1}
  n14 = NamedRerouteUsage()
  n15 = TextureSample(UVs=TextureCoordinate() {uTiling=2}) {samplerSource=SSM_Wrap_WorldGroupSettings, samplerType=SAMPLERTYPE_Normal, texture=/VehicleCosmetics/Wheels/_SharedTextures/FX/T_Zebra_N}
  n16 = AppendVector(A=@n15.R, B=@n15.R)
  n17 = NamedRerouteUsage()
  n18 = MaterialFunctionCall(BaseNormal=NamedRerouteUsage(), AdditionalNormal=AppendVector(A=Multiply(A=@n16, B=@n17) {constB=0.6}, B=@n15.B)) {materialFunction=&BlendAngleCorrectedNormals}
  n19 = NamedRerouteUsage()
  FunctionOutput(None=NamedRerouteDeclaration(None=LinearInterpolate(A=@n14, B=@n18.Result, Alpha=@n19)) {name=Blended N, nodeColor=<rgba 0 0.1562 1 1>}) {outputName=Normal, sortPriority=2}
  n20 = NamedRerouteUsage()
  n21 = NamedRerouteUsage()
  FunctionOutput(None=LinearInterpolate(Alpha=Add(A=@n20, B=@n21)) {constB=0.5}) {outputName=Base Spec Mask, sortPriority=-1}
  NamedRerouteDeclaration(None=FunctionInput() {bUsePreviewValueAsDefault=true, inputName=Normal, sortPriority=2, previewValue=<0 0 1 1>}) {name=Normal, nodeColor=<rgba 0.5 0.5 1 1>}
  n22 = FunctionInput() {bUsePreviewValueAsDefault=true, inputName=Masks, inputType=FunctionInput_Vector4, sortPriority=2, previewValue=<1 1 1 1>}
  NamedRerouteDeclaration(None=ComponentMask(None=@n22) {r=true}) {name=Mask R, nodeColor=<rgba 1 0 0.1875 1>}
  n23 = FunctionInput() {bUsePreviewValueAsDefault=true, inputName=Normal Detail Scale, inputType=FunctionInput_Scalar, sortPriority=4, previewValue=<0.6 0 0 1>}
  NamedRerouteDeclaration(None=@n23) {name=Normal Detail Scale, nodeColor=<rgba 0.5029 0.5029 1 1>}
  NamedRerouteDeclaration(None=ComponentMask(None=@n22) {a=true}) {name=Mask A, nodeColor=<rgba 1 1 1 1>}
  NamedRerouteDeclaration(None=ComponentMask(None=@n22) {b=true}) {name=Mask B, nodeColor=<rgba 0 0 1 1>}
  NamedRerouteDeclaration(None=FunctionInput() {bUsePreviewValueAsDefault=true, inputName=Base Color, previewValue=<1 1 1 1>}) {name=Base Color, nodeColor=<rgba 1 1 0 1>}
  n24 = FunctionInput() {bUsePreviewValueAsDefault=true, inputName=Emission Color B Int, inputType=FunctionInput_Scalar, sortPriority=5, previewValue=<3 0 0 1>}
  NamedRerouteDeclaration(None=@n24) {name=Emission Color B Int, nodeColor=<rgba 0.7991 0 1 1>}
  n25 = FunctionInput() {bUsePreviewValueAsDefault=true, inputName=Pan Speed, inputType=FunctionInput_Scalar, sortPriority=4, previewValue=<0.7 0 0 1>}
  NamedRerouteDeclaration(None=@n25) {name=Pan Speed, nodeColor=<rgba 0.3333 1 0 1>}
  n26 = FunctionInput() {bUsePreviewValueAsDefault=true, inputName=Emission Color B, sortPriority=5, previewValue=<1 0.1683 0.7011 1>}
  NamedRerouteDeclaration(None=@n26) {name=Emission Color B, nodeColor=<rgba 0.7991 0 1 1>}
 ]}
