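# MF_FlipXYValues is a material function in the shared vehicle wheel materials that optionally flips the sign of a Vector2 input per axis.
# It multiplies the input by the MultiplyX and MultiplyY parameters, which default to flipping X only.
# A rounded, saturated FlipValues parameter drives a lerp between original and multiplied values, acting as a binary toggle without a static switch.
# All three scalar parameters share the ShaderFlipValues group.
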MaterialFunction /VehicleCosmetics/Wheels/_SharedMaterials/MF_FlipXYValues
graph {nodes [
  n1 = FunctionInput() {desc=Used to flip Vector2 values on the x axis, inputName=ValueToFlip, inputType=FunctionInput_Vector2}
  n2 = AppendVector(
   A=ScalarParameter() {defaultValue=-1, desc="-1 = Flip, 1 = Default", group=ShaderFlipValues, parameterName=MultiplyX}
   B=ScalarParameter() {defaultValue=1, desc="-1 = Flip, 1 = Default", group=ShaderFlipValues, parameterName=MultiplyY}
  )
  n3 = Round(None=ScalarParameter() {desc="1 = FlipX, 0 = Default", group=ShaderFlipValues, parameterName=FlipValues})
  FunctionOutput(None=LinearInterpolate(A=@n1, B=Multiply(A=@n1, B=@n2), Alpha=Saturate(None=@n3)))
 ]}
