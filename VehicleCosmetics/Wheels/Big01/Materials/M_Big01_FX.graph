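# Translucent, unlit, two-sided effect material for the Big01 cosmetic vehicle wheel, usable on skeletal meshes.
# Flipped UVs are panned and distorted by noise textures to drive layered noise and energy samples, tinted by a Color parameter.
# Two scrolling mask samples, one noise-distorted, combine with a cubed UV gradient to mask that animated pattern.
# Several bias, scale and multiply nodes are labelled as corrections absent from "Prime".
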Material /VehicleCosmetics/Wheels/Big01/Materials/M_Big01_FX
graph {
 properties {
  bAutomaticallySetUsageInEditor=false
  bUsedWithMeshDeformer=true
  bUsedWithSkeletalMesh=true
  blendMode=BLEND_Translucent
  shadingModel=MSM_Unlit
  twoSided=true
  usedShadingModels=Unlit
 }
 nodes [
  n1 = Multiply(A=VectorParameter() {parameterName=Color, defaultValue=<rgba 1 1 1 1>}.RGB, B=Constant() {r=1})
  n2 = MaterialFunctionCall!(ValueToFlip=TextureCoordinate() {uTiling=2, vTiling=2}) {materialFunction=/VehicleCosmetics/Wheels/_SharedMaterials/MF_FlipXYValues}
  n3 = MaterialFunctionCall(ValueToFlip=TextureCoordinate() {uTiling=2, vTiling=2}) {materialFunction=&MF_FlipXYValues}
  n4 = Add(
   A=TextureSample(UVs=Panner(Coordinate=@n2.Result) {speedX=-0.5, speedY=0.5}) {texture=/VehicleCosmetics/Wheels/_SharedTextures/FX/T_Noise_Heightmap}.RGB
   B=TextureSample(UVs=Panner(Coordinate=@n3.Result) {speedX=0.5, speedY=-1}) {texture=&T_Noise_Heightmap}.RGB
  )
  n5 = Add(
   A=Multiply(A=ComponentMask(None=@n4) {g=true, r=true}) {constB=0.1}
   B=MaterialFunctionCall(ValueToFlip=TextureCoordinate() {uTiling=2}) {materialFunction=&MF_FlipXYValues}.Result
  )
  n6 = Multiply(A=TextureSample(UVs=Panner(Coordinate=@n5) {speedX=0.25, speedY=0.65}) {texture=/VehicleCosmetics/Wheels/_SharedTextures/FX/T_NoiseText}.RGB) {constB=5}
  n7 = TextureCoordinate() {uTiling=3}
  n8 = MaterialFunctionCall(ValueToFlip=TextureCoordinate() {uTiling=3, vTiling=2}) {materialFunction=&MF_FlipXYValues}
  n9 = ComponentMask(None=TextureSample(UVs=Panner(Coordinate=@n8.Result) {speedX=-0.8, speedY=-0.8}) {texture=&T_NoiseText}.RGB) {g=true, r=true}
  n10 = Panner(Coordinate=Add(
    A=MaterialFunctionCall(ValueToFlip=@n7) {materialFunction=&MF_FlipXYValues}.Result
    B=Multiply(A=@n9) {constB=0.08}
   )) {speedX=0.25, speedY=0.8}
  n11 = Multiply(A=Multiply(A=@n1, B=TextureSample(UVs=@n10) {texture=/VehicleCosmetics/Wheels/_SharedTextures/FX/T_Energy}.RGB)) {constB=10}
  n12 = Constant() {r=0.5}
  n13 = ConstantBiasScale(None=Add(A=Multiply(A=@n1, B=@n6), B=Desaturation(None=@n11, Fraction=@n12))) {bias=-0.5, desc=Correctional node. Not in Prime., scale=4}
  Multiply(A=@n1, B=@n13)
  n14 = TextureCoordinate() {uTiling=2}
  n15 = TextureSample(UVs=Panner(Coordinate=MaterialFunctionCall(ValueToFlip=@n14) {materialFunction=&MF_FlipXYValues}.Result) {speedX=0.5}) {samplerType=SAMPLERTYPE_LinearColor, texture=/VehicleCosmetics/Wheels/Big01/Textures/T_Big01_FX_Masks}
  n16 = Panner(Coordinate=MaterialFunctionCall(ValueToFlip=TextureCoordinate()) {materialFunction=&MF_FlipXYValues}.Result) {speedY=0.5}
  n17 = Multiply(A=ComponentMask(None=TextureSample(UVs=@n16) {samplerSource=SSM_Wrap_WorldGroupSettings, texture=/VehicleCosmetics/Wheels/_SharedTextures/FX/T_Windtunnel_Noise}.RGB) {g=true, r=true}) {constB=0.01}
  n18 = MaterialFunctionCall(ValueToFlip=TextureCoordinate() {uTiling=2}) {materialFunction=&MF_FlipXYValues}
  n19 = TextureSample(UVs=Panner(Coordinate=Add(A=@n17, B=@n18.Result)) {speedX=0.1}) {samplerType=SAMPLERTYPE_LinearColor, texture=&T_Big01_FX_Masks}
  n20 = Saturate(None=Multiply(A=Multiply(A=@n15.G, B=@n19.G)) {constB=0.1, desc=Correctional node. Not in Prime.})
  n21 = TextureCoordinate()
  n22 = OneMinus(None=ComponentMask(None=MaterialFunctionCall(ValueToFlip=@n21) {materialFunction=&MF_FlipXYValues}.Result) {g=true})
  n23 = Multiply(A=@n22, B=@n22)
  n24 = Saturate(None=ConstantBiasScale(None=Multiply(A=@n22, B=@n23)) {bias=-0.4, desc=Correctional node. Not in Prime., scale=0.2})
  Saturate(None=Multiply(A=@n13, B=Multiply(A=@n20, B=@n24)))
 ]
}
